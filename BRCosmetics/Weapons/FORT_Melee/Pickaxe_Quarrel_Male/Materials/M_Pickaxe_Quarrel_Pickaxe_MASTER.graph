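# Master material for the Quarrel pickaxe in BRCosmetics, adding a timed emissive glow for swings to the shared character material stack.
# The glow fades linearly over SwingGlowDuration from SwingGlowTime, a custom primitive data parameter compared against Time, tinting an emissive texture with SwingGlowColor.
# It blends onto the character PreFX and CharacterSkinFX layers via MatLayerBlend_Emissive, then passes through character and cinematic PostFX functions.
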Material /BRCosmetics/Weapons/FORT_Melee/Pickaxe_Quarrel_Male/Materials/M_Pickaxe_Quarrel_Pickaxe_MASTER
graph {
 properties {
  bAutomaticallySetUsageInEditor=false
  bUseMaterialAttributes=true
  bUsedWithClothing=true
  bUsedWithMeshDeformer=true
  bUsedWithMorphTargets=true
  bUsedWithSkeletalMesh=true
  pixelDepthOffsetMode=PDOM_Legacy
  refractionMethod=RM_IndexOfRefraction
  shadingModel=MSM_FromMaterialExpression
  subsurfaceProfile=&SS_HeroSkin_02
  usedShadingModels=Default Lit | Subsurface | Clear Coat | Subsurface Profile | Cloth | Eye
 }
 nodes [
  n1 = MaterialFunctionCall() {materialFunction=&BaseFN_CharacterFunction_PreFX}
  n2 = Subtract(A=Time(), B=ScalarParameter() {bUseCustomPrimitiveData=true, defaultValue=-10, parameterName=SwingGlowTime})
  n3 = ScalarParameter() {defaultValue=0.2, parameterName=SwingGlowDuration}
  n4 = OneMinus(None=Saturate(None=Divide(A=@n2, B=@n3)))
  n5 = Multiply(
   A=TextureSample() {texture=/BRCosmetics/Weapons/FORT_Melee/Pickaxe_Quarrel_Male/Textures/T_Pickaxe_Quarrel_Male_E}.RGB
   B=VectorParameter() {parameterName=SwingGlowColor, defaultValue=<rgba 5 4.1669 0.7031 1>}.RGB
  ) {constA=1}
  n6 = MaterialFunctionCall(
   AO=@n1.AO
   Distance Blend=@n1.AO
   Fresnel=@n1.AO
   Gmap=@n1.AO
   Pre Skinned Local Position=@n1.AO
   Material=MaterialFunctionCall!(
    Material=MaterialFunctionCall(
     PreskinnedPosition=@n1.Pre skinned Local Position
     Gmap=@n1.Pre skinned Local Position
     MaterialAttributes=@n1.Pre skinned Local Position
     AO=@n1.Pre skinned Local Position
    ) {materialFunction=&CharacterSkinFX}.Result
    Emissive=Multiply(A=@n4, B=@n5) {constA=1}
   ) {materialFunction=/Engine/Functions/MaterialLayerFunctions/MatLayerBlend_Emissive}.Blended Material
  ) {materialFunction=&BaseFN_CharacterFunction_PostFX}
  MaterialFunctionCall(In=@n6.Result) {materialFunction=&CINE_MF_CharPostFX}
 ]
}
