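# Physics asset for the Quarrel Male pickaxe mesh, defining simulation bodies and joints for its two dangling straps.
# Strap A (six bones) and strap B (five bones) each form a chain of small capsules hanging from the b_AR_Root capsule.
# Each link is constrained to its parent with the same moderate swing and tighter twist limits, and capsules widen toward the strap ends.
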PhysicsAsset /BRCosmetics/Weapons/FORT_Melee/Pickaxe_Quarrel_Male/Meshes/Quarrel_Male_Axe_Physics
bodies {
 b_AR_Root [{shapeType=Capsule, length=3, radius=1, shapeName=b_AR_Root_capsule}]
 dyn_strapA1 [{shapeType=Capsule, length=3, radius=1, shapeName=dyn_strapA1_capsule, center=<5 0 0>, rotation=<rot -90 0.0199 -0.0201>}]
 dyn_strapA2 [{shapeType=Capsule, length=2, radius=1, shapeName=dyn_strapA2_capsule, center=<5 0 0>, rotation=<rot -90 0.0199 -0.0201>}]
 dyn_strapA3 [{shapeType=Capsule, length=2, radius=1, shapeName=dyn_strapA3_capsule, center=<5 0 0>, rotation=<rot -90 0.0199 -0.0201>}]
 dyn_strapA4 [{shapeType=Capsule, length=2, radius=1.5, shapeName=dyn_strapA4_capsule, center=<5 0 0>, rotation=<rot -90 2.5307 -2.5309>}]
 dyn_strapA5 [{shapeType=Capsule, length=2, radius=2, shapeName=dyn_strapA5_capsule, center=<5 0 0>, rotation=<rot -90 5.0512 -5.0512>}]
 dyn_strapA6 [{shapeType=Capsule, length=2, radius=2, shapeName=dyn_strapA6_capsule, center=<5 0 0>, rotation=<rot -90 0.0199 -0.0201>}]
 dyn_strapB1 [{shapeType=Capsule, length=2, radius=1, shapeName=dyn_strapB1_capsule, center=<5 0 0>, rotation=<rot -90 27.9383 -27.9383>}]
 dyn_strapB2 [{shapeType=Capsule, length=2, radius=1, shapeName=dyn_strapB2_capsule, center=<5 0 0>, rotation=<rot -90 27.9383 -27.9383>}]
 dyn_strapB3 [{shapeType=Capsule, length=2, radius=1, shapeName=dyn_strapB3_capsule, center=<5 0 0>, rotation=<rot -90 27.9383 -27.9383>}]
 dyn_strapB4 [{shapeType=Capsule, length=2, radius=1.5, shapeName=dyn_strapB4_capsule, center=<5 0 0>, rotation=<rot -90 27.9383 -27.9383>}]
 dyn_strapB5 [{shapeType=Capsule, length=2, radius=2, shapeName=dyn_strapB5_capsule, center=<5 0 0>, rotation=<rot -90 27.9383 -27.9383>}]
}
constraints [
 {bone1Name=dyn_strapA1, bone2Name=b_AR_Root, swing1LimitDegrees=30, swing2LimitDegrees=30, twistLimitDegrees=10}
 {bone1Name=dyn_strapA2, bone2Name=dyn_strapA1, swing1LimitDegrees=30, swing2LimitDegrees=30, twistLimitDegrees=10}
 {bone1Name=dyn_strapA3, bone2Name=dyn_strapA2, swing1LimitDegrees=30, swing2LimitDegrees=30, twistLimitDegrees=10}
 {bone1Name=dyn_strapA4, bone2Name=dyn_strapA3, swing1LimitDegrees=30, swing2LimitDegrees=30, twistLimitDegrees=10}
 {bone1Name=dyn_strapA5, bone2Name=dyn_strapA4, swing1LimitDegrees=30, swing2LimitDegrees=30, twistLimitDegrees=10}
 {bone1Name=dyn_strapA6, bone2Name=dyn_strapA5, swing1LimitDegrees=30, swing2LimitDegrees=30, twistLimitDegrees=10}
 {bone1Name=dyn_strapB1, bone2Name=b_AR_Root, swing1LimitDegrees=30, swing2LimitDegrees=30, twistLimitDegrees=10}
 {bone1Name=dyn_strapB2, bone2Name=dyn_strapB1, swing1LimitDegrees=30, swing2LimitDegrees=30, twistLimitDegrees=10}
 {bone1Name=dyn_strapB3, bone2Name=dyn_strapB2, swing1LimitDegrees=30, swing2LimitDegrees=30, twistLimitDegrees=10}
 {bone1Name=dyn_strapB4, bone2Name=dyn_strapB3, swing1LimitDegrees=30, swing2LimitDegrees=30, twistLimitDegrees=10}
 {bone1Name=dyn_strapB5, bone2Name=dyn_strapB4, swing1LimitDegrees=30, swing2LimitDegrees=30, twistLimitDegrees=10}
]
